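; Parent gameplay ability for NPCs that makes a night Day Sequence volume follow the AI pawn, modifying time of day around it.
; On activation it spawns the volume at the pawn, or logs and ends the ability if spawning fails.
; A looping timer, set by a scalable-float interval, teleports the volume to the pawn's location instead of attaching it.
; Agent death and ability end both clear the timer and destroy the volume.
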(event Ability|EventActivateAbility
  (|CallParentFunction)
  (bind _returnvalue (Game|SpawnActorfromClass "/NPCLibrary/Abilities/TODModifier/BP_DaySequenceVolume_Night.BP_DaySequenceVolume_Night_C" (Math|Transform|MakeTransform (Transformation|GetActorLocation (Variables|InternalData|GetAIPawn))) "AlwaysSpawn" "MultiplyWithRoot" (Variables|InternalData|GetAIPawn) (Variables|InternalData|GetAIPawn)))
  (bind _output_get (Variables|InternalData|SetBP_TOD _returnvalue))
  (Utilities|IsValid _output_get
    (:Is%20Valid
      (bind _returnvalue_2 (Utilities|Time|SetTimerbyEvent (EventDispatchers|CreateEvent :!event Custom|UpdateTODVolume) (ScalableFloat|GetValueatLevel (Variables|Default|GetUpdateInterval)) true))
      (Variables|InternalData|SetHNDL_UpdateTODVolume _returnvalue_2)
      (Death|BindEventtoOnAgentDiedEvent (Variables|InternalData|GetAIController) (EventDispatchers|CreateEvent :!event Custom|AIPawnDied)))
    (:Is%20Not%20Valid
      (Fortnite|FortShippingLog self (Utilities|String|ToString(Text) (Utilities|Text|FormatText "NSLOCTEXT(\"[DBE432982390B90433F6721D176329F4]\", \"918B06914CC7E0F9C901509FD16E56E1\", \":: Failed to spawn Day Sequence Volume for AI: {a}. Cancelling Ability\")" (Utilities|GetDisplayName (Variables|InternalData|GetAIPawn)))) true)
      (Ability|EndAbility))))

(event Custom|UpdateTODVolume
  (Transformation|SetActorLocation (Variables|InternalData|GetBP_TOD) (Transformation|GetActorLocation (Variables|InternalData|GetAIPawn)) false true))

(event Custom|AIPawnDied
  (params (BotController :type Fort%20Athena%20AIBot%20Controller%20Object%20Reference) (AgentPawn :type Fort%20Pawn%20Object%20Reference))
  (Utilities|Time|ClearandInvalidateTimerbyHandle (Variables|InternalData|GetHNDL_UpdateTODVolume))
  (Actor|DestroyActor (Variables|InternalData|GetBP_TOD)))

(event Ability|EventOnEndAbility
  (params (bWasCancelled :type Boolean))
  (bind _bp_tod (Variables|InternalData|GetBP_TOD))
  (bind _aicontroller (Variables|InternalData|GetAIController))
  (|CallParentFunction bWasCancelled)
  (Utilities|Time|ClearandInvalidateTimerbyHandle (Variables|InternalData|GetHNDL_UpdateTODVolume))
  (Utilities|IsValid _bp_tod
    (:Is%20Valid
      (Actor|DestroyActor _bp_tod))
    (:Is%20Not%20Valid
      _))
  (Utilities|IsValid _aicontroller
    (:Is%20Valid
      (Death|UnbindEventfromOnAgentDiedEvent _aicontroller (EventDispatchers|CreateEvent :!event Custom|AIPawnDied)))))
